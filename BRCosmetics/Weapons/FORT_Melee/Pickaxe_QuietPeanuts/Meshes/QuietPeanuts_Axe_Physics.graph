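# Physics asset for the QuietPeanuts pickaxe that gives its dangling tassels capsule bodies and joints for physical motion.
# A handle capsule anchors a tiny master body, which carries the tassel root through a twist-locked joint.
# The root branches through two more master bodies into separate four-link chains.
# Master-to-parent joints rotate freely, while the chain links allow wide swing and very little twist.
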PhysicsAsset /BRCosmetics/Weapons/FORT_Melee/Pickaxe_QuietPeanuts/Meshes/QuietPeanuts_Axe_Physics
bodies {
 dyn_tassel_01 [{shapeType=Capsule, length=0.1, radius=1.43, shapeName=dyn_tassel_01_capsule, rotation=<rot 90 0 0>}]
 dyn_tassel_02 [{shapeType=Capsule, length=1.9, radius=0.4, shapeName=dyn_tassel_02_capsule, center=<1.1224 0 0>, rotation=<rot 90 0 0>}]
 dyn_tassel_02_MSTR [{shapeType=Capsule, length=0.1, radius=0.1, shapeName=dyn_tassel_02_MSTR_capsule, rotation=<rot 0 -90 0>}]
 dyn_tassel_03 [{shapeType=Capsule, length=2.3, radius=0.4, shapeName=dyn_tassel_03_capsule, center=<1.4252 0 0>, rotation=<rot 90 0 0>}]
 dyn_tassel_04 [{shapeType=Capsule, length=2.3, radius=0.35, shapeName=dyn_tassel_04_capsule, center=<2.581 0 0>, rotation=<rot 90 0 0>}]
 dyn_tassel_05 [{shapeType=Capsule, length=2.3, radius=0.3, shapeName=dyn_tassel_05_capsule, center=<2.3991 0 0>, rotation=<rot 90 0 0>}]
 dyn_tassel_06 [{shapeType=Capsule, length=1.6917, radius=0.6295, shapeName=dyn_tassel_06_capsule, center=<1.5127 0 0>, rotation=<rot 90 0 0>}]
 dyn_tassel_06_MSTR [{shapeType=Capsule, length=0.1, radius=0.1, shapeName=dyn_tassel_06_MSTR_capsule, rotation=<rot 0 -90 0>}]
 dyn_tassel_07 [{shapeType=Capsule, length=3.1958, radius=0.59, shapeName=dyn_tassel_07_capsule, center=<1.7537 0 0>, rotation=<rot 90 0 0>}]
 dyn_tassel_08 [{shapeType=Capsule, length=4.0022, radius=0.7098, shapeName=dyn_tassel_08_capsule, center=<2.5375 0 0>, rotation=<rot 90 0 0>}]
 dyn_tassel_09 [{shapeType=Capsule, length=2.9718, radius=0.8205, shapeName=dyn_tassel_09_capsule, center=<2.8693 0 0>, rotation=<rot 90 0 0>}]
 dyn_tassel_MSTR [{shapeType=Capsule, length=0.1, radius=0.1, shapeName=dyn_tassel_MSTR_capsule, rotation=<rot 0 -90 0>}]
 handle [{shapeType=Capsule, length=30, radius=6, shapeName=handle_capsule, center=<-2.5089 0 6.135>}]
}
constraints [
 {
  bone1Name=dyn_tassel_MSTR
  bone2Name=handle
  swing1LimitDegrees=45
  swing1Motion=Free
  swing2LimitDegrees=45
  swing2Motion=Free
  twistLimitDegrees=45
  twistMotion=Free
 }
 {bone1Name=dyn_tassel_01, bone2Name=dyn_tassel_MSTR, swing1LimitDegrees=45, swing2LimitDegrees=70, twistLimitDegrees=1, twistMotion=Locked}
 {bone1Name=dyn_tassel_02, bone2Name=dyn_tassel_02_MSTR, swing1LimitDegrees=77, swing2LimitDegrees=77, twistLimitDegrees=3}
 {bone1Name=dyn_tassel_03, bone2Name=dyn_tassel_02, swing1LimitDegrees=77, swing2LimitDegrees=77, twistLimitDegrees=3}
 {bone1Name=dyn_tassel_04, bone2Name=dyn_tassel_03, swing1LimitDegrees=77, swing2LimitDegrees=77, twistLimitDegrees=3}
 {bone1Name=dyn_tassel_05, bone2Name=dyn_tassel_04, swing1LimitDegrees=77, swing2LimitDegrees=77, twistLimitDegrees=3}
 {bone1Name=dyn_tassel_06, bone2Name=dyn_tassel_06_MSTR, swing1LimitDegrees=70, swing2LimitDegrees=70, twistLimitDegrees=3}
 {bone1Name=dyn_tassel_07, bone2Name=dyn_tassel_06, swing1LimitDegrees=70, swing2LimitDegrees=70, twistLimitDegrees=3}
 {bone1Name=dyn_tassel_08, bone2Name=dyn_tassel_07, swing1LimitDegrees=70, swing2LimitDegrees=70, twistLimitDegrees=3}
 {bone1Name=dyn_tassel_09, bone2Name=dyn_tassel_08, swing1LimitDegrees=70, swing2LimitDegrees=70, twistLimitDegrees=3}
 {
  bone1Name=dyn_tassel_06_MSTR
  bone2Name=dyn_tassel_01
  swing1LimitDegrees=45
  swing1Motion=Free
  swing2LimitDegrees=45
  swing2Motion=Free
  twistLimitDegrees=45
  twistMotion=Free
 }
 {
  bone1Name=dyn_tassel_02_MSTR
  bone2Name=dyn_tassel_01
  swing1LimitDegrees=45
  swing1Motion=Free
  swing2LimitDegrees=45
  swing2Motion=Free
  twistLimitDegrees=45
  twistMotion=Free
 }
]
